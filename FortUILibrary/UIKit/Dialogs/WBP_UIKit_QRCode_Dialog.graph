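# Activatable dialog that shows a QR code beside title, description and image, with an optional link to a website.
# On activation it loads the QR image from QROverride's URL, or QRImage if empty, then feeds a new dialog viewmodel into Dialog_Base.
# Close is always added; Visit Site appears only if the URL can launch, gaining an external-link icon.
# Button index 0 deactivates the dialog; index 1 launches the URL.
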WidgetBlueprint /FortUILibrary/UIKit/Dialogs/WBP_UIKit_QRCode_Dialog
graphs [
 Add External Link Button Icon {nodes [
   Add External Link Button Icon() {is=title K2Node_FunctionEntry}
   n1 = Get ButtonWidgets!(self:WBP UIKit Dialog Internal Object Reference=Get Dialog_Base() -> (Dialog_Base:WBP UIKit Dialog Internal Object Reference)) -> (ButtonWidgets:Array of WBP UIKit Button CTA Base Object References)
   > Branch(Condition=Is Valid Index!(TargetArray:Array of WBP UIKit Button CTA Base Object References=@n1, IndexToTest:int=1) -> (ReturnValue:bool))
   n2 = "Get (a copy)"(Array:Array of WBP UIKit Button CTA Base Object References=@n1, Dimension 1=1) -> (Output:WBP UIKit Button CTA Base Object Reference)
   > Set HasIcon!(execute, HasIcon:bool=true, self:WBP UIKit Button CTA Base Object Reference=@n2) -> (then, Output_Get:bool)
   > Set IconBeforeText!(execute, IconBeforeText:bool=true, self:WBP UIKit Button CTA Base Object Reference=@n2) -> (then, Output_Get:bool)
   > SetIcon!(
    execute
    self:WBP UIKit Button CTA Base Object Reference=@n2
    Icon:Object Reference=/FortUILibrary/Foundation/Buttons/Textures/T_Icon_ExternalLink
   ) -> (then)
  ]}
 EventGraph {nodes [
   n1 = On Dialog_Base Button Clicked!() -> (OutputDelegate, then, ButtonIndex:int) {is=title K2Node_CustomEvent}
   > n2 = Switch on Int!(execute:Exec, Selection:int=@n1.ButtonIndex) -> (0:Exec, 1:Exec) {bHasDefaultPin=false}
   > DeactivateWidget()
   LaunchURL!(execute=@n2.1, URL:String=Get URL() -> (URL:String)) -> (then)
   Event Construct()
   > Bind Event to Dialog Button Clicked Dispatcher(self:WBP UIKit Dialog Internal Object Reference=Get Dialog_Base(), Delegate=Create Event()) {is=title K2Node_AddDelegate}
   Event On Activated()
   > n3 = Construct Object from Class(Class=/Script/FortniteUI.UIKitDialogViewModel) -> (ReturnValue:UI Kit Dialog Viewmodel Object Reference)
   > UpdateOverrides()
   > UpdateDescription!(execute, self, ViewmodelInstance:UI Kit Dialog Viewmodel Object Reference=@n3) -> (then)
   > SetDialogData!(execute, self:WBP UIKit Dialog Internal Object Reference=Get Dialog_Base(), Data:UI Kit Dialog Viewmodel Object Reference=@n3) -> (then)
   > Add External Link Button Icon() {is=title K2Node_CallFunction}
  ]}
 UpdateDescription {nodes [
   n1 = UpdateDescription!() -> (then, ViewmodelInstance:UI Kit Dialog Viewmodel Object Reference) {is=title K2Node_FunctionEntry}
   > Set DialogInfoText!(execute, DialogInfoText:Text=Get DescriptionText() -> (DescriptionText:Text), self:UI Kit Dialog Viewmodel Object Reference=@n1) -> (then, Output_Get:Text)
   > Set DialogTitle!(execute, DialogTitle:Text=Get TitleText() -> (TitleText:Text), self:UI Kit Dialog Viewmodel Object Reference=@n1) -> (then, Output_Get:Text)
   > SetDialogImage!(
    execute
    self:UI Kit Dialog Viewmodel Object Reference=@n1
    InImage:Texture 2D Object Reference=Get DialogImage() -> (DialogImage:Texture 2D Object Reference)
   ) -> (then)
   > n2 = Sequence!(execute) -> (then_0, then_1, then_2:Exec)
   > n3 = Construct Object from Class(Class=/Script/FortniteUI.UIKitDialogButtonViewModel) -> (ReturnValue:UI Kit Dialog Button View Model Object Reference)
   > Set Text(Text="LOCTEXT(\"Close\")", self:UI Kit Dialog Button View Model Object Reference=@n3)
   > Set ActionType!(execute, ActionType:EFortDialogResult Enum=Confirmed, self:UI Kit Dialog Button View Model Object Reference=@n3) -> (then, Output_Get:EFortDialogResult Enum=Confirmed)
   > Set DataTable!(
    execute
    DataTable:Data Table Soft Object Reference=&InputActionDataTable
    self:UI Kit Dialog Button View Model Object Reference=@n3
   ) -> (then, Output_Get:Data Table Soft Object Reference)
   > Set InputAction!(execute, InputAction:Name=Confirm, self:UI Kit Dialog Button View Model Object Reference=@n3) -> (then, Output_Get:Name=None)
   > Add(
    TargetArray:Array of UI Kit Dialog Button View Model Object References=Get ButtonList() -> (ButtonList:Array of UI Kit Dialog Button View Model Object References)
    NewItem:"UI Kit Dialog Button View Model Object Reference (by ref)"=@n3
   ) {is=title K2Node_CallArrayFunction}
   CanLaunchURL!(execute=@n2.then_1, URL:String=Get URL()) -> (then, ReturnValue:bool)
   > n4 = Construct Object from Class(Class=&UIKitDialogButtonViewModel)
   > Set Text(Text="LOCTEXT(\"Visit Site\")", self=@n4)
   > Set ActionType(ActionType=Ignored, self=@n4) -> (Output_Get=Confirmed)
   > Set DataTable(DataTable=&InputActionDataTable, self=@n4)
   > Set InputAction(InputAction=Invite, self=@n4) -> (Output_Get=None)
   > Add(TargetArray=Get ButtonList(), NewItem=@n4)
   Set Buttons!(
    execute=@n2.then_2
    Buttons:Array of UI Kit Dialog Button View Model Object References=Get ButtonList()
    self:UI Kit Dialog Viewmodel Object Reference=@n1
   ) -> (then, Output_Get:Array of UI Kit Dialog Button View Model Object References)
  ]}
 UpdateOverrides {nodes [
   UpdateOverrides() {is=title K2Node_FunctionEntry}
   n1 = Get QROverride() -> (QROverride:String)
   n2 = IsEmpty(InString=@n1)
   > n3 = Branch(Condition=@n2)
   n4 = Get QRLazyImage() -> (QRLazyImage:Fort Lazy Image Object Reference)
   > SetBrushFromLazyTexture!(
    execute
    self:Common Lazy Image Object Reference=@n4
    LazyTexture:"Texture 2D Soft Object Reference (by ref)"=Get QRImage() -> (QRImage:Texture 2D Soft Object Reference)
    bMatchSize:bool=false
   ) -> (then)
   SetBrushFromURL!(
    execute=@n3.else
    self:Fort Lazy Image Object Reference=@n4
    CMSImageURL:String=@n1
    bMatchSize:bool=@n2
    Delay:float=0
    DownloadPriority:ECMSDownloadPriority Enum=Normal
   ) -> (then)
  ]}
]
variables [
 DescriptionText {type=string, default=""}
 DialogImage {type=Texture2D}
 QRImage {type=Texture2D}
 QROverride {type=string, default=""}
 TitleText {type=string, default=""}
 URL {type=string, default=""}
]
widgets {
 info {
  inheritedWidgetCount=0
  namedSlotCount=1
  parentClass=&CommonActivatableWidget
  rootWidgetClass=/FortUILibrary/Systems/Dialog/WBP_UIKit_Dialog_Internal.WBP_UIKit_Dialog_Internal_C
  widgetCount=6
 }
 widgets [
  Dialog_Base {
   widgetClassPath=/FortUILibrary/Systems/Dialog/WBP_UIKit_Dialog_Internal
   bIsVariable=true
   widget=Dialog_Base
   properties {
    useImage=true
    useText=Common Rich Text
    buttons {
     each Button Horizontal Fill=true
     each Button Min Width=300
     each Button Size=Medium
     focuse Button=0
     space Between Buttons=30
     cTA Buttons Types [Loud,Regular]
     group Placement {
      horizontal Alignment=HAlign_Left
      vertical Alignment=VAlign_Center
      grid {column=1, column Span=0, layer=0, row=2, row Span=0, nudge=<0 0>}
      padding=<ltrb 80 64 80 48>
     }
    }
    dialogSIze {
     height Override {enable=false, value=600}
     max Aspect Ratio {enable=false, value=0}
     max Desired Height {enable=true, value=800}
     max Desired Width {enable=false, value=0}
     min Aspect Ratio {enable=false, value=0}
     min Desired Height {enable=true, value=600}
     min Desired Width {enable=false, value=0}
     width Override {enable=true, value=1618}
    }
    imageProperties {
     material=/FortUILibrary/Systems/Dialog/MaterialInstances/MI_UI_DialogImage
     image Size=<512 0>
     placement {
      horizontal Alignment=HAlign_Fill
      vertical Alignment=VAlign_Fill
      grid {column=1, column Span=0, layer=20, row=0, row Span=1, nudge=<0 0>}
      padding=<ltrb 0 0 0 0>
     }
    }
    textProperties {
     auto Wrap Text=true
     justification=Left
     line Height Percentage=1.2
     mobile Font Size Multiplier=1.3
     transform Policy=None
     wrap Text At=0
     color {colorUseRule=UseColor_Specified, specifiedColor=<rgba 1 1 1 0.8>}
     font {
      bForceMonospaced=false
      bMaterialIsStencil=false
      fontMaterial=null
      fontObject=&HeadingNow
      letterSpacing=0
      monospacedWidth=1
      size=15.75
      skewAmount=0
      typefaceFontName=74 Regular
      outlineSettings {
       bApplyOutlineToDropShadows=false
       bMiteredCorners=false
       bSeparateFillAlpha=false
       outlineMaterial=null
       outlineSize=0
       outlineColor=<rgba 0 0 0 1>
      }
     }
     placement {
      horizontal Alignment=HAlign_Fill
      vertical Alignment=VAlign_Top
      grid {column=1, column Span=0, layer=0, row=1, row Span=0, nudge=<0 0>}
      padding=<ltrb 80 0 80 0>
     }
    }
   }
  }
  Overlay_0 {widgetClassPath=&Overlay, namedSlotHost=Dialog_Base, widget=Overlay_0}
  SizeBox {
   widgetClassPath=&SizeBox
   parent=Overlay_0
   widget=SizeBox
   properties {
    bOverride_HeightOverride=true
    bOverride_WidthOverride=true
    heightOverride=288
    maxDesiredHeight=512
    maxDesiredWidth=512
    minDesiredHeight=50
    minDesiredWidth=50
    widthOverride=288
   }
   slot {class=OverlaySlot, properties {horizontalAlignment=HAlign_Center, verticalAlignment=VAlign_Center}}
  }
  Overlay_2 {widgetClassPath=&Overlay, parent=SizeBox, widget=Overlay_2, slot {class=SizeBoxSlot}}
  QR_Backing_Image {
   widgetClassPath=&Image
   parent=Overlay_2
   widget=QR_Backing_Image
   properties {colorAndOpacity=<rgba 0 0 0 0.9>}
   slot {class=OverlaySlot, properties {horizontalAlignment=HAlign_Fill, verticalAlignment=VAlign_Fill, padding=<ltrb 4 4 4 4>}}
  }
  QRLazyImage {
   widgetClassPath=&FortLazyImage
   bIsVariable=true
   parent=Overlay_2
   widget=QRLazyImage
   properties {brush {
     bIsDynamicallyLoaded=false
     drawAs=Image
     imageType=FullColor
     mirroring=NoMirror
     resourceName=None
     resourceObject=null
     tiling=NoTile
     imageSize=<257 257>
     margin=<ltrb 0 0 0 0>
     outlineSettings {
      bUseBrushTransparency=false
      roundingType=HalfHeightRadius
      width=0
      color {colorUseRule=UseColor_Specified, specifiedColor=<rgba 0 0 0 0>}
      cornerRadii=<0 0 0 0>
     }
     tintColor {colorUseRule=UseColor_Specified, specifiedColor=<rgba 1 1 1 1>}
     uVRegion {bIsValid=false, max=<0 0>, min=<0 0>}
    }}
   slot {class=OverlaySlot, properties {horizontalAlignment=HAlign_Center, verticalAlignment=VAlign_Center}}
  }
 ]
}
