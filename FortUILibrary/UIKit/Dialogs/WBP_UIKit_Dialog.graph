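# WBP_UIKit_Dialog is a UI Kit dialog, an activatable CommonUI widget that wraps a single WBP_UIKit_Dialog_Internal with preset background and button layout.
# ApplyDialogVM passes the dialog viewmodel to the inner widget, which determines the widgets used and sets their data.
# The button-click dispatcher is bound on activation and unbound on deactivation.
# A click with a valid index sends that button viewmodel's action type and input action as the dialog result.
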WidgetBlueprint /FortUILibrary/UIKit/Dialogs/WBP_UIKit_Dialog
graphs [
 ApplyDialogVM {nodes [
   n1 = ApplyDialogVM!() -> (then, DialogVM:UI Kit Dialog Viewmodel Object Reference) {is=title K2Node_FunctionEntry}
   n2 = Get Dialog_Internal() -> (Dialog_Internal:WBP UIKit Dialog Internal Object Reference)
   > DetermineWidgetsUsedBasedOnVM!(execute, self:WBP UIKit Dialog Internal Object Reference=@n2, DialogVM:UI Kit Dialog Viewmodel Object Reference=@n1) -> (then)
   > SetDialogData!(execute, self:WBP UIKit Dialog Internal Object Reference=@n2, Data:UI Kit Dialog Viewmodel Object Reference=@n1) -> (then)
  ]}
 EventGraph {nodes [
   n1 = OnButtonClicked!() -> (OutputDelegate, then, ButtonIndex:int) {is=title K2Node_CustomEvent}
   > n2 = Get!(execute:Exec) -> (then:Exec, else:Exec, UIKitDialogViewModel:UI Kit Dialog Viewmodel Object Reference)
   n3 = Get Buttons!(self:UI Kit Dialog Viewmodel Object Reference=@n2) -> (Buttons:Array of UI Kit Dialog Button View Model Object References)
   > Branch(Condition=Is Valid Index!(TargetArray:Array of UI Kit Dialog Button View Model Object References=@n3, IndexToTest:int=@n1.ButtonIndex) -> (ReturnValue:bool))
   n4 = "Get (a copy)"(Array:Array of UI Kit Dialog Button View Model Object References=@n3, Dimension 1=@n1.ButtonIndex) -> (Output:UI Kit Dialog Button View Model Object Reference)
   > SendResult!(
    execute
    self:UI Kit Dialog Viewmodel Object Reference=@n2
    Result:EFortDialogResult Enum=Get ActionType!(self:UI Kit Dialog Button View Model Object Reference=@n4) -> (ActionType:EFortDialogResult Enum=Confirmed)
    ResultName:Name=Get InputAction!(self:UI Kit Dialog Button View Model Object Reference=@n4) -> (InputAction:Name=None)
   ) -> (then)
   Event On Activated()
   > Bind Event to Dialog Button Clicked Dispatcher(self:WBP UIKit Dialog Internal Object Reference=Get Dialog_Internal(), Delegate=@n1) {is=title K2Node_AddDelegate}
   Event On Deactivated()
   > Unbind Event from Dialog Button Clicked Dispatcher(self:WBP UIKit Dialog Internal Object Reference=Get Dialog_Internal(), Delegate=@n1) {is=title K2Node_RemoveDelegate}
  ]}
]
variables []
widgets {
 info {
  inheritedWidgetCount=0
  parentClass=&CommonActivatableWidget
  rootWidgetClass=/FortUILibrary/Systems/Dialog/WBP_UIKit_Dialog_Internal.WBP_UIKit_Dialog_Internal_C
  widgetCount=1
 }
 widgets [Dialog_Internal {
   widgetClassPath=/FortUILibrary/Systems/Dialog/WBP_UIKit_Dialog_Internal
   bIsVariable=true
   widget=Dialog_Internal
   properties {
    defaultInputActionDataTable=&InputActionDataTable
    backgroundProperties {
     material=/FortUILibrary/UIKit/Dialogs/MaterialInstances/MI_UI_DialogBackground
     image Size=<32 32>
     placement {
      horizontal Alignment=HAlign_Fill
      vertical Alignment=VAlign_Fill
      grid {column=0, column Span=2, layer=1, row=0, row Span=0, nudge=<0 0>}
      padding=<ltrb 0 0 0 0>
     }
    }
    buttons {
     each Button Horizontal Fill=false
     each Button Min Width=200
     each Button Size=Medium
     focuse Button=0
     space Between Buttons=30
     cTA Buttons Types [Loud,Regular]
     group Placement {
      horizontal Alignment=HAlign_Fill
      vertical Alignment=VAlign_Bottom
      grid {column=1, column Span=0, layer=0, row=2, row Span=0, nudge=<0 0>}
      padding=<ltrb 80 64 80 48>
     }
    }
   }
  }]
}
